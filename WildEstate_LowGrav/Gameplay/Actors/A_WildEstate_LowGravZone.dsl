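; Wild Estate low-gravity zone actor that applies gameplay effects to player pawns and vehicles inside its box.
; Begin overlap applies the effect, deferring until ASC initialization if needed; end overlap removes it by tag.
; Local players get a time-throttled gameplay alert, and pawns with a ready ASC get a looping camera cue.
; A scalable-float toggle gates both overlaps, so disabling does not clear effects already applied.
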(event EventBeginPlay)

(event Custom|WildEstate_ToggleLowGravZone
  (params (IsEnabled :type Boolean))
  (ScalableFloat|SetValue (Variables|Default|GetLowGravEnabled) (Math|Conversions|ToFloat(Boolean) IsEnabled)))

(event Collision|EventActorBeginOverlap
  (params (OtherActor :type Actor%20Object%20Reference)))

(event EventTick
  (params (DeltaSeconds :type Float%20%28single-precision%29)))

(event OnComponentBeginOverlap(Box)
  (params (OverlappedComponent :type Primitive%20Component%20Object%20Reference) (OtherActor :type Actor%20Object%20Reference) (OtherComp :type Primitive%20Component%20Object%20Reference) (OtherBodyIndex :type Integer) (bFromSweep :type Boolean) (SweepResult :type Hit%20Result%20Structure%20%28by%20ref%29))
  (bind _returnvalue (Utilities|Time|GetRealTimeSeconds))
  (if (ScalableFloat|GetValueasBool (Variables|Default|GetLowGravEnabled))
    (bind _asfort_player_pawn (Utilities|Casting|CastToFortPlayerPawn OtherActor)
      (:then
        (Utilities|FlowControl|Sequence
          (:then_0
            (Networking|CanExecuteCosmeticEvents
              (:True
                (bind _asfort_player_pawn_athena (Utilities|Casting|CastToFortPlayerPawnAthena OtherActor))
                (if (Pawn|IsLocalController (Pawn|GetController _asfort_player_pawn_athena))
                  (if (> (- _returnvalue (Variables|Default|GetLastGameplayAlertTime)) (ScalableFloat|GetValueatLevel (Variables|Default|GetMinTimeBetweenGameplayAlerts)))
                    (GameplayEventRouter|BroadcastStatefulEvent (GameplayEventRouter|GetGlobalEventRouter) (Variables|Default|GetGameplayAlert) 0)
                    (Variables|Default|SetLastGameplayAlertTime _returnvalue))))) ; Am I a client?
          )
          (:then_1
            (Utilities|IsValid (Ability|GetFortAbilitySystemComponent _asfort_player_pawn)
              (:Is%20Valid
                (CallFunction|ApplyGEToASC (Ability|GetFortAbilitySystemComponent _asfort_player_pawn))
                ;[0.8385 0.08157 0.1377 0.5] VFX Camera FX (Leaf FX)
                (Projectile|AddGameplayCue%28Looping%29Local _asfort_player_pawn (Variables|Default|GetGCNL_CameraFX) (Ability|GameplayCue|MakeGameplayCueParameters 0.0)))
              (:Is%20Not%20Valid
                (EventDispatchers|BindEventtoOnAbilitySystemComponentInitialized _asfort_player_pawn (EventDispatchers|CreateEvent :!event Custom|ApplyGEToASC)))))))
      (:CastFailed
        (bind _asfort_athena_vehicle (Utilities|Casting|CastToFortAthenaVehicle OtherActor))
        (GameplayEffects|ApplyGameplayEffectToSelf :self (Ability|GetAbilitySystemComponent _asfort_athena_vehicle) :GameplayEffectClass (Variables|Default|GetVehicleLowGravGE) :Level 0.0)
        (Development|PrintString "VEHICLE COLLIDED !!!")))))

(event OnComponentEndOverlap(Box)
  (params (OverlappedComponent :type Primitive%20Component%20Object%20Reference) (OtherActor :type Actor%20Object%20Reference) (OtherComp :type Primitive%20Component%20Object%20Reference) (OtherBodyIndex :type Integer))
  (if (ScalableFloat|GetValueasBool (Variables|Default|GetLowGravEnabled))
    (bind _asfort_pawn (Utilities|Casting|CastToFortPawn OtherActor)
      (:then
        (GameplayEffects|RemoveActiveEffectswithTags (Ability|GetAbilitySystemComponent _asfort_pawn) (Variables|Default|GetLowGravEffectTag))
        (Development|PrintString "PLAYER STOP COLLIDING !!!")
        ;[0.8385 0.08157 0.1377 0.5] VFX Camera FX (Leaf FX)
        (Projectile|RemoveGameplayCue%28Looping%29Local _asfort_pawn (Variables|Default|GetGCNL_CameraFX) (Ability|GameplayCue|MakeGameplayCueParameters 0.0)))
      (:CastFailed
        (bind _asfort_athena_vehicle (Utilities|Casting|CastToFortAthenaVehicle OtherActor))
        (GameplayEffects|RemoveActiveEffectswithTags (Ability|GetAbilitySystemComponent _asfort_athena_vehicle) (Variables|Default|GetLowGravVehicleEffectTag))
        (Development|PrintString "VEHICLE STOP COLLIDING !!!")))))

(event Custom|ApplyGEToASC
  (params (AbilitySystemComponent :type Fort%20Ability%20System%20Component%20Object%20Reference) (AffectedPawn :type Fort%20Player%20Pawn%20Object%20Reference))
  (GameplayEffects|ApplyGameplayEffectToSelf :self AbilitySystemComponent :GameplayEffectClass (Variables|Default|GetLowGravGE) :Level 0.0))

(fn ConstructionScript)
