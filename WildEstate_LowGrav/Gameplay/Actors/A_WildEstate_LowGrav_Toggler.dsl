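; Level actor for the WildEstate low-gravity feature that switches its World Partition data layer on or off.
; On BeginPlay it waits asynchronously for playlist data to be ready, then evaluates the IsLowGravEnabled scalable float as a bool.
; It sets the EDL_WildEstate_LowGrav data layer to Activated when true and Unloaded otherwise, non-recursively.
; Overlap, Tick and the construction script are empty.
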(event EventBeginPlay
  (Class|FortAsyncActionCurrentPlaylistReady|PlaylistDataReadyAsync
    (:OnReady
      (if (ScalableFloat|GetValueasBool (Variables|Default|GetIsLowGravEnabled))
        (bind _returnvalue (WorldPartition|GetDataLayerManager))
        (DataLayers|SetDataLayerInstanceRuntimeState _returnvalue (DataLayers|GetDataLayerInstancefromAsset _returnvalue "/WildEstate_LowGrav/Levels/EDL_WildEstate_LowGrav.EDL_WildEstate_LowGrav") "Activated" false)
        (else
          (bind _returnvalue_2 (WorldPartition|GetDataLayerManager))
          (DataLayers|SetDataLayerInstanceRuntimeState _returnvalue_2 (DataLayers|GetDataLayerInstancefromAsset _returnvalue_2 "/WildEstate_LowGrav/Levels/EDL_WildEstate_LowGrav.EDL_WildEstate_LowGrav") "Unloaded" false))))))

(event Collision|EventActorBeginOverlap
  (params (OtherActor :type Actor%20Object%20Reference)))

(event EventTick
  (params (DeltaSeconds :type Float%20%28single-precision%29)))

(fn ConstructionScript)
