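; Blueprint prop actor for the low-gravity FlipStream in WildEstate, built on a parent class that supplies its base behavior.
; On BeginPlay it runs the parent logic, applies the NeonBurn FlipStream WPO material instance to the static mesh and turns off custom depth rendering.
; It then destroys itself if the ShouldSpawn scalable float evaluates to false, letting data control whether the prop exists.
; The construction script only calls the parent.
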(event EventBeginPlay
  (|CallParentFunction)
  (Rendering|Material|SetMaterial (Variables|Debug|GetStaticMeshComponent) 0 "/WildEstate/Gameplay/NeonBurn/FlipStream/Assets/Materials/MI_NeonBurn_FlipStream_A_WPO.MI_NeonBurn_FlipStream_A_WPO")
  (Rendering|SetRenderCustomDepth (Variables|Debug|GetStaticMeshComponent) false)
  (if (not (ScalableFloat|GetValueasBool (Variables|Default|GetShouldSpawn)))
    (Actor|DestroyActor)))

(fn ConstructionScript
  (|CallParentFunction))
