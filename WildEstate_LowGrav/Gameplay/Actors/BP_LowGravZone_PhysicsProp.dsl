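; Physics prop actor for the Wild Estate low-gravity zone, with the graph holding only interaction and audio cleanup logic.
; BlueprintCanInteract always returns true, allowing any pawn and interaction type.
; When destroyed, it quickly fades out its ambient audio component.
; The construction script is empty, and no gravity or physics behavior appears in this graph.
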(event EventDestroyed
  (Audio|Components|Audio|FadeOut (Variables|Default|GetAmbientAudio) 0.1 0.0 "Linear" :!class /Script/Engine.AudioComponent))

(fn BlueprintCanInteract
  (params (InteractingPawn :type Fort%20Pawn%20Object%20Reference) (InteractionBeingAttempted :type EInteractionBeingAttempted%20Enum :default "FirstInteraction") (InteractionType :type TInteractionType%20Enum :default "IT_NoInteraction"))
  (return true))

(fn ConstructionScript)
