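; Blueprint actor for low-gravity zone debris, judging by its name, that removes itself when spawning is disabled.
; On BeginPlay it reads a ShouldSpawn scalable float as a bool and destroys the actor when that value is false.
; The construction script is empty, and the shown graphs hold no other logic.
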(event EventBeginPlay
  (if (not (ScalableFloat|GetValueasBool (Variables|Components|GetShouldSpawn)))
    (Actor|DestroyActor)))

(fn ConstructionScript)
